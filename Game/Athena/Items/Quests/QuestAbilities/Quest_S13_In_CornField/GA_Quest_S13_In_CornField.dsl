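; Gameplay ability for a Season 13 quest that marks the owning player pawn as being in a cornfield while it remains there.
; On event activation it applies the InCornField gameplay effect to the pawn, stores the handle and starts a looping two-second timer.
; Each tick scans the pawn's overlapping components for the corn tag.
; When none is found, it clears the timer, removes the effect and ends the ability.
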(event Ability|EventActivateAbilityFromEvent
  (params (EventData :type Gameplay%20Event%20Data%20Structure%20%28by%20ref%29))
  (|CallParentFunction EventData)
  (bind _returnvalue (GameplayEffects|ApplyGameplayEffectToSelf :self (Ability|GetAbilitySystemComponent (Variables|GAVars|Data|GetOwningPlayerPawn)) :GameplayEffectClass "/Game/Athena/Items/Quests/QuestAbilities/Quest_S13_In_CornField/GE_Quest_S13_In_CornField.GE_Quest_S13_In_CornField_C" :Level 0.0))
  (Variables|Default|SetHandle_InCorn _returnvalue)
  (bind _returnvalue_2 (Utilities|Time|SetTimerbyEvent (delegate Custom|Check%20If%20In%20Corn) 2.0 true false 2.0))
  (Variables|Default|SetTimer _returnvalue_2))

(event Custom|Check%20If%20In%20Corn
  (bind _owningplayerpawn (Variables|GAVars|Data|GetOwningPlayerPawn))
  (Variables|Default|SetInCorn false)
  (bind _array_element (Utilities|Array|ForEachLoopwithBreak (Collision|GetOverlappingComponents _owningplayerpawn :!class /Script/Engine.Actor))
    (:LoopBody
      (bind _output_get (Variables|Default|SetInCorn (Components|ComponentHasTag _array_element (Variables|Default|Getcorn%20tag))))
      (if _output_get
        _))
    (:Completed
      _))
  (if (not (Variables|Default|GetInCorn))
    (Utilities|Time|ClearandInvalidateTimerbyHandle (Variables|Default|GetTimer))
    (GameplayEffects|RemoveActiveGameplayEffect (Ability|GetAbilitySystemComponent _owningplayerpawn) (Variables|Default|GetHandle_InCorn) -1)
    (Ability|EndAbility)))
